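# A simple Default Lit material in the capture team's camera sequence folder, evidently a green screen surface for capture work.
# It exposes a Color vector parameter defaulting to chroma green and a Roughness scalar parameter, with no texture or math nodes.
# Usage flags cover Nanite, instanced static meshes, geometry collections and static lighting.
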Material /PilgrimCore/CameraManager/Sequences/CaptureTeam/GreenScreen/CT_M_GreenScreen
graph {
 properties {
  bUsedWithGeometryCollections=true
  bUsedWithInstancedStaticMeshes=true
  bUsedWithNanite=true
  bUsedWithStaticLighting=true
  pixelDepthOffsetMode=PDOM_Legacy
  usedShadingModels=Default Lit
 }
 nodes [
  ScalarParameter() {defaultValue=0.6407, parameterName=Roughness}
  VectorParameter() {parameterName=Color, defaultValue=<rgba 0 0.6302 0.0131 1>}
 ]
}
